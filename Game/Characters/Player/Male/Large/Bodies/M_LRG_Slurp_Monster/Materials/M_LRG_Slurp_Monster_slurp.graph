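# Slurp Monster character body material that fakes a glowing, bubble-filled liquid interior, mostly through emissive on a default-lit surface.
# Panning pre-skinned triplanar bubble layers use view-vector offsets for depth, simplified on low quality, forward and mobile.
# Vertex-color spherical harmonics and noise blend inner and outer colors, with faked sun highlights and time-of-day scaling.
# Masked sine waves wobble vertices along normals; an optional FXMask blends in textures, before shared override, de-rez and opacity functions.
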Material /Game/Characters/Player/Male/Large/Bodies/M_LRG_Slurp_Monster/Materials/M_LRG_Slurp_Monster_slurp
graph {
 properties {
  bUseMaterialAttributes=true
  bUsedWithMeshDeformer=true
  bUsedWithMorphTargets=true
  bUsedWithSkeletalMesh=true
  pixelDepthOffsetMode=PDOM_Legacy
  translucencyLightingMode=TLM_Surface
  usedShadingModels=Default Lit
 }
 nodes [
  n1 = ScalarParameter() {defaultValue=25, parameterName=BubbleSize_A}
  n2 = PreSkinnedPosition()
  n3 = VertexInterpolator(VS=@n2)
  n4 = CameraPositionWS()
  n5 = Normalize(VectorInput=Subtract(A=@n3.PS, B=TransformPosition(None=@n4) {transformSourceType=TRANSFORMPOSSOURCE_World}))
  n6 = QualitySwitch(Default=Add(A=@n3.PS, B=Multiply(A=@n5) {constB=12}), Low=@n3.PS)
  n7 = ShadingPathSwitch(Default=@n6, Deferred=@n6, Forward=@n3.PS, Mobile=@n3.PS)
  n8 = Divide(A=@n7, B=Multiply(A=Abs(None=@n1)) {constB=-1})
  n9 = Time()
  n10 = Time()
  n11 = Multiply(A=Add(A=ComponentMask(None=@n7) {b=true}, B=Multiply(A=@n10) {constB=0.5})) {constB=0.05}
  n12 = Multiply(A=Sine(None=Frac(None=@n11))) {constB=0}
  n13 = AppendVector(A=@n12, B=ScalarParameter() {defaultValue=0.025, parameterName=Bubble Speed Y_1})
  n14 = TextureObject() {samplerType=SAMPLERTYPE_Normal, texture=/Game/Characters/Player/Male/Large/Bodies/M_LRG_Slurp_Monster/Textures/T_SlurpBubble_N}
  n15 = TextureSample(UVs=Panner(Coordinate=ComponentMask(None=@n8) {g=true, r=true}, Time=@n9, Speed=@n13), Tex=@n14) {samplerType=SAMPLERTYPE_Normal}
  n16 = VertexInterpolator(VS=PreSkinnedNormal())
  n17 = ScalarParameter() {defaultValue=0.5, parameterName=Blend Sharpness, sliderMax=1}
  n18 = Max(A=Add(A=Abs(None=@n16.PS), B=Multiply(A=@n17) {constB=-0.55}) {constB=-0.55}) {constB=0}
  n19 = Divide(A=@n18, B=DotProduct(A=@n18, B=Constant3Vector() {constant=<rgba 1 1 1 0>}))
  n20 = TextureSample(UVs=Panner(Coordinate=ComponentMask(None=@n8) {b=true, r=true}, Time=@n9, Speed=@n13), Tex=@n14) {samplerType=SAMPLERTYPE_Normal}
  n21 = Multiply(
   A=MaterialFunctionCall(X=@n20.R, Z=@n20.R) {materialFunction=&MakeFloat3}.Result
   B=ComponentMask(None=@n19) {g=true}
  )
  Add(
   A=Multiply(
    A=MaterialFunctionCall(X=@n15.R, Y=@n15.R) {materialFunction=&MakeFloat3}.Result
    B=ComponentMask(None=@n19) {b=true}
   )
   B=@n21
  )
  n22 = Multiply(A=@n5) {constB=15}
  n23 = Add(A=@n3.PS) {constB=20}
  n24 = QualitySwitch(Default=Add(A=Add(A=@n3.PS, B=@n22)) {constB=20}, Low=@n23)
  n25 = ShadingPathSwitch(Default=@n24, Deferred=@n24, Forward=@n23, Mobile=@n23)
  n26 = Divide(B=Multiply(A=Abs(None=@n25)) {constB=-1})
  n27 = AppendVector(A=Multiply(A=@n12) {constB=-0.5}, B=ScalarParameter() {defaultValue=0.05, parameterName=Bubble Speed Y_2})
  n28 = TextureSample(UVs=Panner(Coordinate=ComponentMask(None=@n26) {g=true, r=true}, Time=@n9, Speed=@n27), Tex=@n14) {samplerType=SAMPLERTYPE_Normal}
  n29 = TextureSample(UVs=Panner(Coordinate=ComponentMask(None=@n26) {b=true, r=true}, Time=@n9, Speed=@n27), Tex=@n14) {samplerType=SAMPLERTYPE_Normal}
  n30 = Multiply(
   A=MaterialFunctionCall(X=@n29.R, Z=@n29.R) {materialFunction=&MakeFloat3}.Result
   B=ComponentMask(None=@n19) {g=true}
  )
  Add(
   A=Multiply(
    A=MaterialFunctionCall(X=@n28.R, Y=@n28.R) {materialFunction=&MakeFloat3}.Result
    B=ComponentMask(None=@n19) {b=true}
   )
   B=@n30
  )
  n31 = VertexColor()
  MaterialFunctionCall(X=@n31.R, Y=@n31.R, Z=@n31.R, A=@n31.R) {materialFunction=&MakeFloat4}
  n32 = Constant3Vector()
  n33 = TextureSampleParameter2D() {parameterName=Diffuse, texture=&T_BlacksRGB}
  n34 = TextureSampleParameter2D() {parameterName=FXMask, texture=&T_BlacksRGB}
  n35 = StaticBoolParameter() {parameterName=Use FXmask and textures}
  n36 = ScalarParameter() {defaultValue=0.5, parameterName=Specular}
  n37 = TextureSampleParameter2D() {parameterName=Specular, samplerType=SAMPLERTYPE_Masks, texture=&T_BlackMask}
  n38 = ScalarParameter() {defaultValue=0.2, parameterName=Roughness}
  n39 = CollectionParameter() {
   collection=&FortniteMaterialParameters
   parameterName=SunlightDirectionalVector
  }
  n40 = ComponentMask(None=@n39) {b=true, g=true, r=true}
  n41 = TextureSample(UVs=Panner(Coordinate=ComponentMask(None=@n26) {b=true, g=true}, Time=@n9, Speed=@n27), Tex=@n14) {samplerType=SAMPLERTYPE_Normal}
  n42 = Multiply(
   A=MaterialFunctionCall(Y=@n41.R, Z=@n41.R) {materialFunction=&MakeFloat3}.Result
   B=ComponentMask(None=@n19) {r=true}
  )
  n43 = TextureSample(UVs=Panner(Coordinate=ComponentMask(None=@n8) {b=true, g=true}, Time=@n9, Speed=@n13), Tex=@n14) {samplerType=SAMPLERTYPE_Normal}
  n44 = Multiply(
   A=MaterialFunctionCall(Y=@n43.R, Z=@n43.R) {materialFunction=&MakeFloat3}.Result
   B=ComponentMask(None=@n19) {r=true}
  )
  n45 = TextureObject() {samplerType=SAMPLERTYPE_Masks, texture=/Game/Characters/Player/Male/Large/Bodies/M_LRG_Slurp_Monster/Textures/T_SlurpBubble}
  n46 = Constant() {r=2}
  n47 = MaterialFunctionCall(Float3=MaterialFunctionCall!(
    TextureObject=@n45
    TextureSize=@n1
    WorldPosition=@n7
    Export Float 4
    In=@n13
    World Space Normal=@n16.PS
    ProjectionTransitionContrast=@n46
   ) {materialFunction=/Game/Athena/SafeZone/Preskinned_WP_Texture_ScalePan}.XY Texture) {materialFunction=&BreakOutFloat3Components}
  n48 = Add(A=@n47.B, B=Add(A=@n47.R, B=@n47.R))
  n49 = MaterialFunctionCall() {materialFunction=&Fresnel_Function}
  n50 = MaterialFunctionCall!(
   WorldPosition=@n7
   World Space Normal=@n16.PS
   UVSize_Front=@n1
   UVSize_Side=@n1
   UVSize_Top=@n1
   ProjectionTransitionContrast=Constant() {r=0.1}
  ) {materialFunction=/Game/Packages/Fortress_SharedMaterials/Base_Material_Functions/WorldAlignedTexture_UVsMasks}
  n51 = Saturate(None=Subtract(A=@n47.R, B=@n50.Mask_Top))
  n52 = MaterialFunctionCall(
   TextureObject=@n45
   TextureSize=ScalarParameter() {defaultValue=30, parameterName=BubbleSize_A_1}
   WorldPosition=@n25
   In=@n27
   World Space Normal=@n16.PS
   ProjectionTransitionContrast=@n46
  ) {materialFunction=&Preskinned_WP_Texture_ScalePan}
  n53 = Saturate(None=Subtract(
    A=MaterialFunctionCall(Float3=@n52.XY Texture) {materialFunction=&BreakOutFloat3Components}.G
    B=@n50.Mask_Top
   ))
  n54 = Saturate(None=Multiply(A=OneMinus(None=@n49.Result), B=Add(A=@n51, B=@n53)))
  n55 = LinearInterpolate(A=Constant3Vector(), B=LinearInterpolate(A=Add(A=@n30, B=@n42), B=Add(A=@n21, B=@n44), Alpha=Saturate(None=@n48)), Alpha=@n54)
  n56 = Transform(None=Normalize(VectorInput=Add(A=@n16.PS, B=@n55))) {transformSourceType=TRANSFORMSOURCE_Local}
  n57 = Multiply(A=Multiply(A=DotProduct(A=@n40, B=@n56)) {constB=2}, B=@n56)
  n58 = DotProduct(
   A=Normalize(VectorInput=Subtract(A=@n57, B=@n40))
   B=Multiply(A=MaterialFunctionCall() {materialFunction=&CameraDirectionVector}.Result)
  )
  n59 = Power(Base=Max(A=@n58) {constB=0}, Exp=Constant() {r=5})
  n60 = ScalarParameter() {defaultValue=0.15, parameterName=Bubble_BackSpecular_Scale}
  n61 = Power(Base=Multiply(A=@n58) {constB=-1}, Exp=Constant() {r=50})
  n62 = ScalarParameter() {defaultValue=0.55, parameterName=Bubble_HighlightSpecular_Scale}
  n63 = Add(A=Saturate(None=Multiply(A=@n59, B=@n60) {constB=0.15}), B=Saturate(None=Multiply(A=@n61, B=@n62) {constB=0.55}))
  n64 = Multiply(
   A=Subtract(A=ComponentMask(None=@n3.PS) {b=true}, B=ScalarParameter() {defaultValue=125, parameterName=OuterColor_TransitionStart}) {constB=125}
   B=ScalarParameter() {defaultValue=0.015, parameterName=OuterColor_TransitionScale}
  ) {constB=0.015}
  n65 = ComponentMask(None=Saturate(None=MaterialFunctionCall() {materialFunction=/Engine/Functions/Engine_MaterialFunctions02/Texturing/3DSandMayaUVCoordinates}.UVChannel1)) {r=true}
  n66 = OneMinus(None=@n65)
  n67 = Multiply(A=@n50.UV_Front) {constB=0.25}
  n68 = Multiply(A=Time()) {constB=0.05}
  n69 = AppendVector(A=ComponentMask(None=@n67) {r=true}, B=Add(A=ComponentMask(None=@n67) {g=true}, B=@n68))
  n70 = Multiply(A=@n50.UV_Side) {constB=0.25}
  n71 = AppendVector(A=ComponentMask(None=@n70) {r=true}, B=Add(A=ComponentMask(None=@n70) {g=true}, B=@n68))
  n72 = Multiply(A=LinearInterpolate(
    A=TextureSample(UVs=@n69) {constMipValue=0, samplerType=SAMPLERTYPE_Grayscale, texture=&LowResBlurredNoise}.R
    B=TextureSample(UVs=@n71) {constMipValue=0, samplerType=SAMPLERTYPE_Grayscale, texture=&LowResBlurredNoise}.R
    Alpha=@n50.Mask_Side
   )) {constB=2}
  n73 = Power(Base=Saturate(None=@n72)) {constExponent=4}
  n74 = Add(A=@n51, B=@n53)
  n75 = Multiply(A=@n5) {constB=15}
  n76 = Add(A=@n3.PS) {constB=166}
  n77 = QualitySwitch(Default=Add(A=Add(A=@n3.PS, B=@n75)) {constB=166}, Low=@n76)
  n78 = MaterialFunctionCall(
   TextureObject=@n45
   TextureSize=ScalarParameter() {defaultValue=15, parameterName=BubbleSize_A_2}
   WorldPosition=ShadingPathSwitch(Default=@n77, Deferred=@n77, Forward=@n76, Mobile=@n76)
   In=AppendVector(A=ScalarParameter() {parameterName=Bubble Speed X}, B=ScalarParameter() {defaultValue=0.12, parameterName=Bubble Speed Y_3})
   World Space Normal=@n16.PS
   ProjectionTransitionContrast=@n46
  ) {materialFunction=&Preskinned_WP_Texture_ScalePan}
  n79 = Saturate(None=Subtract(
    A=MaterialFunctionCall(Float3=@n78.XY Texture) {materialFunction=&BreakOutFloat3Components}.B
    B=@n50.Mask_Top
   ))
  n80 = QualitySwitch(Default=Add(A=@n74, B=Multiply(A=@n79) {constB=0.5}), Low=@n74)
  n81 = ShadingPathSwitch(Default=@n80, Deferred=@n80, Forward=@n74, Mobile=@n74)
  n82 = MaterialFunctionCall!(ReflectionVector, ReflectionTexture) {materialFunction=/Engine/Functions/Engine_MaterialFunctions01/Reflections/ViewAlignedReflection}
  n83 = ScalarParameter() {defaultValue=1, parameterName=ViewDependantDistortion}
  n84 = ScalarParameter() {defaultValue=0.025, parameterName=water_SpeedX}
  n85 = ScalarParameter() {defaultValue=0.05, parameterName=water_SpeedY}
  n86 = Rotator(Coordinate=Panner(Coordinate=Multiply(A=@n82.UVs, B=@n83), Speed=AppendVector(A=@n84, B=@n85)), Time=Time()) {speed=0.0001}
  n87 = Multiply(
   A=Multiply(
    A=TextureSample(UVs=@n86) {constMipValue=0, samplerType=SAMPLERTYPE_Grayscale, texture=&LowResBlurredNoise}.G
    B=Multiply(A=@n73, B=@n81)
   )
   B=OneMinus(None=Fresnel() {baseReflectFraction=0, exponent=1})
  )
  n88 = AppendVector(A=@n31, B=@n31)
  n89 = VectorParameter() {parameterName=SHMax, defaultValue=<rgba 1.25 0.13 0.18 0.13>}
  n90 = VectorParameter() {parameterName=SHMin, defaultValue=<rgba 1 -0.16 -0.17 -0.15>}
  n91 = AppendVector(A=@n90.RGB, B=@n90.RGB)
  n92 = Subtract(A=AppendVector(A=@n89.RGB, B=@n89.RGB), B=@n91)
  n93 = Multiply(A=MaterialFunctionCall(X=@n31.G, Y=@n31.G, Z=@n31.G) {materialFunction=&MakeFloat3}.Result)
  MaterialExpressionStaticSwitchParameter_1 = StaticSwitchParameter(True=Add(A=Multiply(A=@n88, B=@n92), B=@n91), False=AppendVector(A=Add(A=@n31.R), B=Add(A=@n93) {constB=-0.5})) {parameterName=UseCustomSH Fit}
  n94 = ComponentMask(None=MaterialFunctionCall() {materialFunction=&CameraDirectionVector}.Result) {b=true, g=true, r=true}
  n95 = Multiply(A=Multiply(A=@n74, B=@n72), B=ScalarParameter() {defaultValue=0.25, parameterName=SHBubbleScale})
  MaterialExpressionTransform_0 = Transform(None=Normalize(VectorInput=LinearInterpolate(A=@n94, B=@n56, Alpha=@n95))) {transformSourceType=TRANSFORMSOURCE_World, transformType=TRANSFORM_Local}
  n96 = Custom!(SH=@MaterialExpressionStaticSwitchParameter_1, Direction=@MaterialExpressionTransform_0) {
   code="//SH decode,  Direction = LightDir, SH = 2band coefficients. (encoded in Y-up)\r\nDirection = normalize(Direction);\r\nfloat4 shEvalDirection;\r\nshEvalDirection.x = 0.282095;\r\nshEvalDirection.y =0.488603 * Direction.z;\r\nshEvalDirection.z = 0.488603 * Direction.y;\r\nshEvalDirection.w =-0.488603 * Direction.x;\r\nreturn saturate(dot(SH, shEvalDirection));"
   description=SphericalHarmonicsEval
   outputType=CMOT_Float1
   inputs [
    {
     inputName=SH
     input {expression=&self:MaterialExpressionStaticSwitchParameter_1}
    }
    {
     inputName=Direction
     input {expression=&self:MaterialExpressionTransform_0}
    }
   ]
  }
  n97 = MaterialFunctionCall(
   Input=@n96
   Input Low=ScalarParameter() {parameterName=SHContrast_LowValue}
   Input High=ScalarParameter() {defaultValue=1, parameterName=SHContrast_HighValue}
   Target Low=Constant()
   Target High=Constant() {r=1}
  ) {materialFunction=&RemapValueRange}
  n98 = Power(Base=Saturate(None=OneMinus(None=@n97.Result)), Exp=ScalarParameter() {defaultValue=1, parameterName=SHPow})
  n99 = QualitySwitch(Default=Add(A=Add(A=@n98, B=@n87), B=Multiply(A=@n66) {constB=0.0333}), Low=@n98)
  n100 = LinearInterpolate(
   A=VectorParameter() {parameterName=InnerColor, defaultValue=<rgba 0 0.0892 0.25 1>}.RGB
   B=LinearInterpolate(
    A=VectorParameter() {parameterName=OuterColorA, defaultValue=<rgba 0.0125 0.724 0.6663 1>}.RGB
    B=VectorParameter() {parameterName=OuterColorB, defaultValue=<rgba 0 0.5903 0.75 0>}.RGB
    Alpha=Saturate(None=@n64)
   )
   Alpha=ShadingPathSwitch(Default=@n99, Deferred=@n99, Forward=@n98, Mobile=@n98)
  )
  n101 = VectorParameter() {parameterName=InnerScatteringColor, defaultValue=<rgba 0.0877 0.0586 0.2031 1>}
  n102 = CollectionParameter() {
   collection=&FortniteMaterialParameters
   parameterName=SunlightDirectionalVector
  }
  n103 = ComponentMask(None=@n102) {b=true, g=true, r=true}
  n104 = VertexNormalWS()
  n105 = DotProduct(A=@n103, B=@n104)
  n106 = Subtract(A=Multiply(A=Multiply(A=@n105) {constB=2}, B=@n104), B=@n103)
  n107 = MaterialFunctionCall() {materialFunction=&CameraDirectionVector}
  n108 = Max(A=DotProduct(A=Normalize(VectorInput=@n106), B=Multiply(A=@n107.Result) {constB=-1})) {constB=0}
  n109 = Constant() {r=5}
  n110 = Saturate(None=Multiply(A=Power(Base=@n108, Exp=@n109)) {constB=0.4})
  n111 = Multiply(
   A=MaterialFunctionCall(Value_Morning=Constant() {r=0.66}, Value_Day=Constant() {r=1}, Value_Evening=Constant() {r=0.66}, Value_Night=Constant() {r=0.5}) {materialFunction=&Scalar_Time_of_Day}.Result
   B=Add(A=LinearInterpolate(A=@n100, B=@n101.RGB, Alpha=@n110), B=Multiply(A=@n63, B=@n54))
  )
  n112 = Constant()
  n113 = TextureSampleParameter2D() {parameterName=WPO_Mask, samplerType=SAMPLERTYPE_Masks, texture=&T_WhiteMask}
  n114 = WorldPosition() {worldPositionShaderOffset=WPT_ExcludeAllShaderOffsets}
  n115 = Subtract(A=Frac(None=Multiply(A=@n114.XYZ) {constB=0.0025})) {constB=0.5}
  n116 = MaterialFunctionCall(Float3=Abs(None=Multiply(A=@n115) {constB=2})) {materialFunction=&BreakOutFloat3Components}
  n117 = Add(A=Add(A=@n116.R, B=@n116.R), B=@n116.B)
  n118 = Multiply(A=Time()) {constB=0.25}
  n119 = Sine(None=Subtract(A=Subtract(A=@n117) {constB=2}, B=Frac(None=@n118)))
  n120 = Multiply(A=Multiply(A=Add(A=@n119)) {constB=0.5}, B=ScalarParameter() {defaultValue=1, parameterName=WPO_WorldNoiseScale})
  n121 = MaterialFunctionCall(Float3=Multiply(A=@n2) {constB=0.02}) {materialFunction=&BreakOutFloat3Components}
  n122 = Add(A=@n121.R, B=@n121.R)
  n123 = Subtract(A=Time(), B=ScalarParameter() {parameterName=Slurp_WPO_random_seed})
  n124 = Subtract(A=Subtract(A=Add(A=@n122, B=@n121.B)) {constB=2}, B=Frac(None=Multiply(A=@n123) {constB=0.6}))
  n125 = Multiply(A=Add(A=Sine(None=@n124))) {constB=0.5}
  n126 = Multiply(A=Multiply(A=VertexNormalWS(), B=Add(A=@n120, B=@n125)), B=@n65)
  n127 = SetMaterialAttributes!(
   MaterialAttributes
   Base Color=StaticSwitch(True=LinearInterpolate(A=@n32, B=@n33.RGB, Alpha=@n34.R), False=@n32, Value=@n35)
   Specular=StaticSwitch(True=LinearInterpolate(A=@n36, B=@n37.R, Alpha=@n34.R), False=@n36, Value=@n35)
   Roughness=StaticSwitch(True=LinearInterpolate(A=@n38, B=@n37.B, Alpha=@n34.R), False=@n38, Value=@n35)
   Emissive Color=StaticSwitch(True=LinearInterpolate(A=@n111, B=@n112, Alpha=@n34.R), False=@n111, Value=@n35)
   Normal=TextureSampleParameter2D(UVs=TextureCoordinate()) {parameterName=Normals, samplerType=SAMPLERTYPE_Normal, texture=/Engine/EngineMaterials/FlatNormal}.RGB
   World Position Offset=Multiply(A=Multiply(A=@n113.R, B=@n126), B=ScalarParameter() {defaultValue=1, parameterName=WPO_FinalScale})
   Metallic=StaticSwitch(True=LinearInterpolate(A=@n112, B=@n37.G, Alpha=@n34.R), False=@n112, Value=@n35)
  ) {attributeSetTypes [
    #1
    #2
    #3
    #4
    #5
    #6
    #7
   ]}
  MaterialFunctionCall!(Material=MaterialFunctionCall!(
    Fresnel
    Material=MaterialFunctionCall(In=@n127) {materialFunction=&MF_TechArtOverrides}.Result
   ) {materialFunction=/Game/Packages/Fortress_SharedMaterials/Base_Material_Functions/CharacterDeRez_Transition}.Result) {materialFunction=/Game/Packages/Fortress_SharedMaterials/Base_Material_Functions/BaseCharacterShaderFunctions/BaseFN_CharacterOpacity}
  Multiply() {constB=2}
  PixelNormalWS()
 ]
}
